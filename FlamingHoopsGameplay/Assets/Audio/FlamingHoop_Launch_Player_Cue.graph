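# Sound cue for the player launch sound of a Flaming Hoop ring, with Doppler, slight pitch variation and a three-voice concurrency cap.
# An IsLocal branch plays random first-person variants for the local player, otherwise random 3P variants through dedicated third-person attenuation.
# Within the 3P path, an IsEnemy branch mixes similarly named variant sets louder for enemies or an unset parameter, quieter for non-enemies.
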SoundCue /FlamingHoopsGameplay/Assets/Audio/FlamingHoop_Launch_Player_Cue
root=Doppler_0
nodes [
 {id=Attenuation_0, children [Branch_1], properties {attenuationSettings=&/FlamingHoop_Launch_3P_Att}}
 {id=Attenuation_1, children [Mixer_1]}
 {id=Branch_0, children [Attenuation_1, Attenuation_0, Attenuation_0], properties {boolParameterName=IsLocal}}
 {id=Branch_1, children [Mixer_2,Mixer_3,Mixer_2], properties {boolParameterName=IsEnemy}}
 {id=Doppler_0, children [Modulator_0], properties {bUseSmoothing=true}}
 {id=Mixer_1, children [Random_0], properties {inputVolume [0.7]}}
 {id=Mixer_2, children [Random_3], properties {inputVolume [0.9]}}
 {id=Mixer_3, children [Random_4], properties {inputVolume [0.5]}}
 {id=Modulator_0, children [Branch_0], properties {pitchMax=1.1, volumeMax=1, volumeMin=1}}
 {
  id=Random_0
  children [WavePlayer_1,WavePlayer_2,WavePlayer_3,WavePlayer_5]
  properties {bShouldExcludeFromBranchCulling=true, weights [1,1,1,1]}
 }
 {
  id=Random_3
  children [WavePlayer_14, WavePlayer_15, WavePlayer_16, WavePlayer_17]
  properties {bShouldExcludeFromBranchCulling=true, weights [1,1,1,1]}
 }
 {
  id=Random_4
  children [WavePlayer_18, WavePlayer_19, WavePlayer_20, WavePlayer_21]
  properties {bShouldExcludeFromBranchCulling=true, weights [1,1,1,1]}
 }
 {id=WavePlayer_1, properties {soundWaveAssetPtr=&/FlamingHoop_Ring_Launch_Player_01}}
 {id=WavePlayer_14, properties {soundWaveAssetPtr=&/FlamingHoop_Ring_Launch_Player_3P_01}}
 {id=WavePlayer_15, properties {soundWaveAssetPtr=&/FlamingHoop_Ring_Launch_Player_3P_02}}
 {id=WavePlayer_16, properties {soundWaveAssetPtr=&/FlamingHoop_Ring_Launch_Player_3P_03}}
 {id=WavePlayer_17, properties {soundWaveAssetPtr=&/FlamingHoop_Ring_Launch_Player_3P_04}}
 {id=WavePlayer_18, properties {soundWaveAssetPtr=&FlamingHoop_Ring_Launch_Player_3P_01}}
 {id=WavePlayer_19, properties {soundWaveAssetPtr=&FlamingHoop_Ring_Launch_Player_3P_02}}
 {id=WavePlayer_2, properties {soundWaveAssetPtr=&/FlamingHoop_Ring_Launch_Player_02}}
 {id=WavePlayer_20, properties {soundWaveAssetPtr=&FlamingHoop_Ring_Launch_Player_3P_03}}
 {id=WavePlayer_21, properties {soundWaveAssetPtr=&FlamingHoop_Ring_Launch_Player_3P_04}}
 {id=WavePlayer_3, properties {soundWaveAssetPtr=&/FlamingHoop_Ring_Launch_Player_03}}
 {id=WavePlayer_5, properties {soundWaveAssetPtr=&/FlamingHoop_Ring_Launch_Player_04}}
]
properties {
 bOverrideConcurrency=true
 maxDistance=2097152
 soundClassObject=&Fort_Gadgets
 virtualizationMode=Disabled
 volumeMultiplier=0.8
 concurrencyOverrides {maxCount=3}
}
