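# Sound cue for the Flaming Hoop vehicle launch, playing one of four random wave variations per perspective.
# An IsLocal branch selects first-person waves and attenuation, or third-person attenuation with an IsEnemy branch that mixes enemy launches louder than friendly ones.
# The result passes through pitch modulation and smoothed Doppler.
# It uses the gadgets sound class, a concurrency cap and disabled virtualization.
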SoundCue /FlamingHoopsGameplay/Assets/Audio/FlamingHoop_Launch_Vehicle_Cue
root=Doppler_2
nodes [
 {id=Attenuation_2, children [Branch_6], properties {attenuationSettings=&/FlamingHoop_Launch_3P_Att}}
 {id=Attenuation_3, children [Mixer_7], properties {attenuationSettings=&/FlamingHoop_Launch_1P_Att}}
 {id=Branch_5, children [Attenuation_3, Attenuation_2, Attenuation_2], properties {boolParameterName=IsLocal}}
 {id=Branch_6, children [Mixer_8,Mixer_9,Mixer_8], properties {boolParameterName=IsEnemy}}
 {id=Doppler_2, children [Modulator_3], properties {bUseSmoothing=true}}
 {id=Mixer_7, children [Random_6], properties {inputVolume [0.7]}}
 {id=Mixer_8, children [Random_7], properties {inputVolume [0.9]}}
 {id=Mixer_9, children [Random_8], properties {inputVolume [0.5]}}
 {id=Modulator_3, children [Branch_5], properties {pitchMax=1.1, volumeMax=1, volumeMin=1}}
 {
  id=Random_6
  children [WavePlayer_30, WavePlayer_31, WavePlayer_32, WavePlayer_33]
  properties {bShouldExcludeFromBranchCulling=true, weights [1,1,1,1]}
 }
 {
  id=Random_7
  children [WavePlayer_34, WavePlayer_35, WavePlayer_36, WavePlayer_37]
  properties {bShouldExcludeFromBranchCulling=true, weights [1,1,1,1]}
 }
 {
  id=Random_8
  children [WavePlayer_38, WavePlayer_39, WavePlayer_40, WavePlayer_41]
  properties {bShouldExcludeFromBranchCulling=true, weights [1,1,1,1]}
 }
 {id=WavePlayer_30, properties {soundWaveAssetPtr=&/FlamingHoop_Ring_Launch_Player_01}}
 {id=WavePlayer_31, properties {soundWaveAssetPtr=&/FlamingHoop_Ring_Launch_Player_02}}
 {id=WavePlayer_32, properties {soundWaveAssetPtr=&/FlamingHoop_Ring_Launch_Player_03}}
 {id=WavePlayer_33, properties {soundWaveAssetPtr=&/FlamingHoop_Ring_Launch_Player_04}}
 {id=WavePlayer_34, properties {soundWaveAssetPtr=&/FlamingHoop_Ring_Launch_Player_3P_01}}
 {id=WavePlayer_35, properties {soundWaveAssetPtr=&/FlamingHoop_Ring_Launch_Player_3P_02}}
 {id=WavePlayer_36, properties {soundWaveAssetPtr=&/FlamingHoop_Ring_Launch_Player_3P_03}}
 {id=WavePlayer_37, properties {soundWaveAssetPtr=&/FlamingHoop_Ring_Launch_Player_3P_04}}
 {id=WavePlayer_38, properties {soundWaveAssetPtr=&FlamingHoop_Ring_Launch_Player_3P_01}}
 {id=WavePlayer_39, properties {soundWaveAssetPtr=&FlamingHoop_Ring_Launch_Player_3P_02}}
 {id=WavePlayer_40, properties {soundWaveAssetPtr=&FlamingHoop_Ring_Launch_Player_3P_03}}
 {id=WavePlayer_41, properties {soundWaveAssetPtr=&FlamingHoop_Ring_Launch_Player_3P_04}}
]
properties {
 bOverrideConcurrency=true
 maxDistance=16000
 soundClassObject=&Fort_Gadgets
 virtualizationMode=Disabled
 volumeMultiplier=0.8
 concurrencyOverrides {maxCount=3}
}
