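# Looping ambient Sound Cue for the flaming hoop ring's exhaust pipe (B variant), spatialized by its attenuation asset with per-owner concurrency limits.
# A Mixer layers a slightly pitched-down pipe loop with a volume fade-in envelope over rattle one-shots.
# Rattles are picked at random from 18 waves and repeat after random delays, with modulated pitch and volume.
# Quality Level nodes silence the whole cue at the last two levels.
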SoundCue /FlamingHoopsGameplay/Assets/Audio/FlamingHoop_Ring_Ambient_Exhaust_B_Loop_Cue
root=QualityLevel_1
nodes [
 {id=Delay_0, children [QualityLevel_0], properties {delayMax=4, delayMin=2}}
 {
  id=Enveloper_0
  children [WavePlayer_2]
  properties {pitchCurve {editorCurveData {keys [{time=0}]}}, volumeCurve {editorCurveData {keys [<at 0 0>,<at 1 1>]}}}
 }
 {id=Looping_0, children [Delay_0]}
 {id=Mixer_0, children [Modulator_1,Modulator_2], properties {inputVolume [1,0.8]}}
 {id=Modulator_1, children [Enveloper_0], properties {pitchMax=0.9, pitchMin=0.9, volumeMax=1, volumeMin=1}}
 {id=Modulator_2, children [Looping_0], properties {pitchMax=0.2, pitchMin=1, volumeMax=1, volumeMin=0.5}}
 {id=QualityLevel_0, children [Random_0,Random_0,Random_0,null]}
 {id=QualityLevel_1, children [Mixer_0,Mixer_0,null,null]}
 {
  id=Random_0
  children [
   WavePlayer_3
   WavePlayer_4
   WavePlayer_5
   WavePlayer_6
   WavePlayer_7
   WavePlayer_8
   WavePlayer_9
   WavePlayer_10
   WavePlayer_11
   WavePlayer_12
   WavePlayer_13
   WavePlayer_14
   WavePlayer_15
   WavePlayer_16
   WavePlayer_17
   WavePlayer_18
   WavePlayer_19
   WavePlayer_20
  ]
  properties {bShouldExcludeFromBranchCulling=true, weights [1,1,1,1,1,1,1,1,1,1,1,1,1,1,1,1,1,1]}
 }
 {id=WavePlayer_10, properties {soundWaveAssetPtr=&/FlamingHoop_Ambient_Exhaust_Pipe_Rattle_08}}
 {id=WavePlayer_11, properties {soundWaveAssetPtr=&/FlamingHoop_Ambient_Exhaust_Pipe_Rattle_09}}
 {id=WavePlayer_12, properties {soundWaveAssetPtr=&/FlamingHoop_Ambient_Exhaust_Pipe_Rattle_10}}
 {id=WavePlayer_13, properties {soundWaveAssetPtr=&/FlamingHoop_Ambient_Exhaust_Pipe_Rattle_11}}
 {id=WavePlayer_14, properties {soundWaveAssetPtr=&/FlamingHoop_Ambient_Exhaust_Pipe_Rattle_12}}
 {id=WavePlayer_15, properties {soundWaveAssetPtr=&/FlamingHoop_Ambient_Exhaust_Pipe_Rattle_13}}
 {id=WavePlayer_16, properties {soundWaveAssetPtr=&/FlamingHoop_Ambient_Exhaust_Pipe_Rattle_14}}
 {id=WavePlayer_17, properties {soundWaveAssetPtr=&/FlamingHoop_Ambient_Exhaust_Pipe_Rattle_15}}
 {id=WavePlayer_18, properties {soundWaveAssetPtr=&/FlamingHoop_Ambient_Exhaust_Pipe_Rattle_16}}
 {id=WavePlayer_19, properties {soundWaveAssetPtr=&/FlamingHoop_Ambient_Exhaust_Pipe_Rattle_17}}
 {id=WavePlayer_2, properties {bLooping=true, soundWaveAssetPtr=&/FlamingHoop_Ambient_Exhaust_Pipe_Loop_04}}
 {id=WavePlayer_20, properties {soundWaveAssetPtr=&/FlamingHoop_Ambient_Exhaust_Pipe_Rattle_18}}
 {id=WavePlayer_3, properties {soundWaveAssetPtr=&/FlamingHoop_Ambient_Exhaust_Pipe_Rattle_01}}
 {id=WavePlayer_4, properties {soundWaveAssetPtr=&/FlamingHoop_Ambient_Exhaust_Pipe_Rattle_02}}
 {id=WavePlayer_5, properties {soundWaveAssetPtr=&/FlamingHoop_Ambient_Exhaust_Pipe_Rattle_03}}
 {id=WavePlayer_6, properties {soundWaveAssetPtr=&/FlamingHoop_Ambient_Exhaust_Pipe_Rattle_04}}
 {id=WavePlayer_7, properties {soundWaveAssetPtr=&/FlamingHoop_Ambient_Exhaust_Pipe_Rattle_05}}
 {id=WavePlayer_8, properties {soundWaveAssetPtr=&/FlamingHoop_Ambient_Exhaust_Pipe_Rattle_06}}
 {id=WavePlayer_9, properties {soundWaveAssetPtr=&/FlamingHoop_Ambient_Exhaust_Pipe_Rattle_07}}
]
properties {
 attenuationSettings=&/FlamingHoop_Ambient_Exhaust_Pipe_Att
 bOverrideConcurrency=true
 maxDistance=2350
 soundClassObject=&Fort_Gadgets
 volumeMultiplier=0.9
 concurrencyOverrides {bLimitToOwner=true, maxCount=3}
}
