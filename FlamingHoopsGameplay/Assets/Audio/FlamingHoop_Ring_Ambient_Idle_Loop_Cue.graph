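# Sound cue for the looping idle ambience of a flaming hoop ring, in the Fort_Gadgets sound class.
# A Mixer blends a looping machine hum with a looping fire sound, then feeds a Modulator at the root.
# The fire layer sits behind a quality-level switch that fills only the first two of four slots, so it drops out at the others.
# Close attenuation and an owner-limited, single-instance PreventNew concurrency override control playback.
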SoundCue /FlamingHoopsGameplay/Assets/Audio/FlamingHoop_Ring_Ambient_Idle_Loop_Cue
root=Modulator_0
nodes [
 {id=Mixer_1, children [QualityLevel_0, WavePlayer_1], properties {inputVolume [1,1]}}
 {id=Modulator_0, children [Mixer_1]}
 {id=QualityLevel_0, children [WavePlayer_0,WavePlayer_0,null,null]}
 {id=WavePlayer_0, properties {bLooping=true, soundWaveAssetPtr=&/FlamingHoop_Ambient_Fire_Loop_A}}
 {id=WavePlayer_1, properties {bLooping=true, soundWaveAssetPtr=&/FlamingHoop_Ambient_Machine_Loop_A}}
]
properties {
 attenuationSettings=&/FlamingHoop_Ambient_Close_Att
 bOverrideConcurrency=true
 maxDistance=3800
 soundClassObject=&Fort_Gadgets
 volumeMultiplier=0.7
 concurrencyOverrides {bLimitToOwner=true, maxCount=1, resolutionRule=PreventNew}
}
