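# Sound cue for the player's flaming hoop ring launch, classed as a gadget sound with its own attenuation.
# A mixer layers one of four random launch push waves with a quieter recharge-end wave delayed by just over a second, both through pitch modulators.
# The first mixer input is empty.
# Concurrency allows one instance per owner, stopping the oldest, and virtualization is disabled.
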SoundCue /FlamingHoopsGameplay/Assets/Audio/FlamingHoop_Ring_Launch_Player_Cue
root=Mixer_1
nodes [
 {id=Delay_1, children [Modulator_3], properties {delayMax=1.4, delayMin=1.3}}
 {id=Delay_3, children [Random_2]}
 {id=Mixer_1, children [null,Modulator_5,Delay_1], properties {inputVolume [0.1,1,0.6]}}
 {id=Modulator_3, children [WavePlayer_7], properties {pitchMax=1.2, volumeMax=1, volumeMin=1}}
 {id=Modulator_5, children [Delay_3], properties {volumeMax=1, volumeMin=1}}
 {
  id=Random_2
  children [WavePlayer_12, WavePlayer_13, WavePlayer_14, WavePlayer_15]
  properties {bShouldExcludeFromBranchCulling=true, weights [1,1,1,1]}
 }
 {id=WavePlayer_12, properties {soundWaveAssetPtr=&/FlamingHoop_Ring_Launch_Push_01}}
 {id=WavePlayer_13, properties {soundWaveAssetPtr=&/FlamingHoop_Ring_Launch_Push_02}}
 {id=WavePlayer_14, properties {soundWaveAssetPtr=&/FlamingHoop_Ring_Launch_Push_03}}
 {id=WavePlayer_15, properties {soundWaveAssetPtr=&/FlamingHoop_Ring_Launch_Push_04}}
 {id=WavePlayer_7, properties {soundWaveAssetPtr=&/FlamingHoop_Ring_Recharge_End_02}}
]
properties {
 attenuationSettings=&/FlamingHoop_Ring_Launch_Att
 bOverrideConcurrency=true
 maxDistance=16000
 soundClassObject=&Fort_Gadgets
 virtualizationMode=Disabled
 volumeMultiplier=0.65
 concurrencyOverrides {bLimitToOwner=true, maxCount=1, resolutionRule=StopOldest}
}
