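# Sound cue named for the end of a flaming hoop ring recharge in the FlamingHoops gameplay plugin, routed to the Fort_Gadgets sound class.
# Its node graph is empty with no root, so the cue references no audio and acts as a silent shell.
# It still carries playback settings: the ring launch attenuation asset, a volume boost, disabled virtualization, and a single-instance-per-owner concurrency override that stops the oldest.
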SoundCue /FlamingHoopsGameplay/Assets/Audio/FlamingHoop_Ring_Recharge_End_Cue
root=null
nodes []
properties {
 attenuationSettings=&/FlamingHoop_Ring_Launch_Att
 bOverrideConcurrency=true
 maxDistance=18000
 soundClassObject=&Fort_Gadgets
 virtualizationMode=Disabled
 volumeMultiplier=1.2
 concurrencyOverrides {bLimitToOwner=true, maxCount=1, resolutionRule=StopOldest}
}
